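# Masked, two-sided Default Lit material, named as the blade contrail effect for the Donut Cup pickaxe.
# A Panner with U and V speed parameters scrolls the UVs shared by the BladeFX texture and its normal map, by default along U only.
# The texture's RGB is multiplied by a light grey Color parameter.
# Scalar parameters provide metal, specular and roughness values, with metalness defaulting high and roughness low.
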Material /BRCosmetics/Weapons/FORT_Melee/Pickaxe_Donut_Cup/FX/M_Contrail_Dount_Cap_BladeFX
graph {
 properties {blendMode=BLEND_Masked, pixelDepthOffsetMode=PDOM_Legacy, twoSided=true, usedShadingModels=Default Lit}
 nodes [
  n1 = Panner(
   Coordinate=TextureCoordinate()
   Speed=AppendVector(A=ScalarParameter() {defaultValue=0.5, parameterName=PanSpeedU}, B=ScalarParameter() {parameterName=PanSpeedV})
  ) {speedX=0.5}
  Multiply(
   A=VectorParameter() {parameterName=Color, defaultValue=<rgba 0.5585 0.5905 0.6198 1>}.RGB
   B=TextureSample(UVs=@n1) {texture=&/T_Pickaxe_Donut_Cup_BladeFX}.RGB
  )
  ScalarParameter() {defaultValue=0.85, parameterName=Metal}
  ScalarParameter() {defaultValue=10, parameterName=Specular}
  ScalarParameter() {defaultValue=0.2, parameterName=Rough}
  TextureSample(UVs=@n1) {samplerType=SAMPLERTYPE_Normal, texture=&/T_Pickaxe_Donut_Cup_BladeFX_N}
 ]
}
